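# Sound cue for the looping ambient sound of Takedown Medallions, played quietly through a shared item ambient-loop attenuation asset in the loot-collect sound class.
# A looping wave feeds a Modulator for slight random pitch and volume variation, then an Enveloper whose volume curve fades in over two seconds.
# Overridden concurrency limits simultaneous voices to a pair with distance-based volume scaling, and the cue's debug flag is on.
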SoundCue /VividRazorMedallionsGameplay/Assets/Audio/SC_TakedownMedallions_Amb
root=Enveloper_0
nodes [
 {
  id=Enveloper_0
  children [Modulator_0]
  properties {pitchCurve {editorCurveData {keys [{time=0}]}}, volumeCurve {editorCurveData {keys [<at 0 0>,<at 2 1>]}}}
 }
 {id=Modulator_0, children [WavePlayer_3], properties {pitchMax=1.005, pitchMin=1, volumeMax=1, volumeMin=0.88}}
 {id=WavePlayer_3, properties {bLooping=true, soundWaveAssetPtr=&/Default_TakedownMedallions_Amb}}
]
properties {
 attenuationSettings=/Game/Sounds/Attenuation/Fort_Item_Ambient_Loop_Mono_Default_Att
 bDebug=true
 bOverrideConcurrency=true
 maxDistance=1250
 soundClassObject=&Fort_Loot_Collect
 volumeMultiplier=0.2
 attenuationOverrides {falloffDistance=512, nonSpatializedRadiusStart=20, stereoSpread=420}
 concurrencyOverrides {maxCount=2, voiceStealReleaseTime=0.3, volumeScale=0.4, volumeScaleAttackTime=0.1, volumeScaleMode=Distance}
}
