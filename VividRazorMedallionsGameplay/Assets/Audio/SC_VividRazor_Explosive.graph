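# Sound cue for the VividRazor explosive sound, playing one of five explosion variations.
# A Random node picks among five equally weighted wave players and feeds a root Modulator left at default settings for pitch and volume variation.
# The cue uses a dedicated explosion attenuation asset and the medium explosives sound class, with a slightly reduced volume multiplier.
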SoundCue /VividRazorMedallionsGameplay/Assets/Audio/SC_VividRazor_Explosive
root=Modulator_0
nodes [
 {id=Modulator_0, children [Random_2]}
 {
  id=Random_2
  children [WavePlayer_0,WavePlayer_1,WavePlayer_2,WavePlayer_3,WavePlayer_4]
  properties {bShouldExcludeFromBranchCulling=true, weights [1,1,1,1,1]}
 }
 {id=WavePlayer_0, properties {soundWaveAssetPtr=&/VividRazor_Explosive_01}}
 {id=WavePlayer_1, properties {soundWaveAssetPtr=&/VividRazor_Explosive_02}}
 {id=WavePlayer_2, properties {soundWaveAssetPtr=&/VividRazor_Explosive_03}}
 {id=WavePlayer_3, properties {soundWaveAssetPtr=&/VividRazor_Explosive_04}}
 {id=WavePlayer_4, properties {soundWaveAssetPtr=&/VividRazor_Explosive_05}}
]
properties {
 attenuationSettings=&/Att_VividRazor_Explosion_new
 maxDistance=19500
 soundClassObject=/Game/Sounds/FortSoundClassesAndModes/Fort_Explosives_Medium_PSM
 volumeMultiplier=0.8
}
