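# Sound cue for picking up a takedown medallion, named for the Armored Convoy and Power Division variant.
# A mixer layers the default medallion pickup SFX with a louder musical stinger that starts after a short fixed delay.
# The cue uses the loot-collect sound class and overrides concurrency so only one instance plays at a time.
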SoundCue /VividRazorMedallionsGameplay/Assets/Audio/SC_VividRazor_TakedownMedallions_Pickup_ArmoredConvoy_PowerDivision
root=Mixer_2
nodes [
 {id=Delay_0, children [WavePlayer_0], properties {delayMax=0.2, delayMin=0.2}}
 {id=Mixer_2, children [WavePlayer_5,Delay_0], properties {inputVolume [1,1.3]}}
 {id=WavePlayer_0, properties {soundWaveAssetPtr=&/MX_S34_Medallion_03_Stinger}}
 {id=WavePlayer_5, properties {soundWaveAssetPtr=&/Default_TakedownMedallions_Pickup_SFX}}
]
properties {
 bOverrideConcurrency=true
 maxDistance=2097152
 soundClassObject=&Fort_Loot_Collect
 volumeMultiplier=0.7692
 concurrencyOverrides {maxCount=1}
}
