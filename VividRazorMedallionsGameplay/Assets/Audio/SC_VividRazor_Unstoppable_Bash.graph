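# Sound cue for the Vivid Razor medallion's Unstoppable bash sound, classed under gadget audio.
# A Random node picks one of six equally weighted bash wave variations and feeds it through a Modulator at the root.
# It uses a player bash-impact attenuation asset, overrides concurrency to cap simultaneous instances at a low count, and disables virtualization.
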SoundCue /VividRazorMedallionsGameplay/Assets/Audio/SC_VividRazor_Unstoppable_Bash
root=Modulator_0
nodes [
 {id=Modulator_0, children [Random_0]}
 {id=Random_0, children [WavePlayer_0,WavePlayer_1,WavePlayer_2,WavePlayer_3,WavePlayer_4,WavePlayer_5], properties {weights [1,1,1,1,1,1]}}
 {id=WavePlayer_0, properties {soundWaveAssetPtr=&/VividRazor_Unstoppable_Bash_01}}
 {id=WavePlayer_1, properties {soundWaveAssetPtr=&/VividRazor_Unstoppable_Bash_02}}
 {id=WavePlayer_2, properties {soundWaveAssetPtr=&/VividRazor_Unstoppable_Bash_03}}
 {id=WavePlayer_3, properties {soundWaveAssetPtr=&/VividRazor_Unstoppable_Bash_04}}
 {id=WavePlayer_4, properties {soundWaveAssetPtr=&/VividRazor_Unstoppable_Bash_05}}
 {id=WavePlayer_5, properties {soundWaveAssetPtr=&/VividRazor_Unstoppable_Bash_06}}
]
properties {
 attenuationSettings=&/Att_VividRazor_Unstoppable_Player_Bash_Impact_Att
 bOverrideConcurrency=true
 maxDistance=14000
 soundClassObject=&Fort_Gadgets
 virtualizationMode=Disabled
 volumeMultiplier=0.8
 concurrencyOverrides {maxCount=2, volumeScaleAttackTime=0.1}
}
