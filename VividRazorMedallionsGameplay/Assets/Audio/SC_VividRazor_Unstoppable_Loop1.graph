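# Looping sound cue for the VividRazor medallion Unstoppable effect, in the gadgets sound class, with concurrency limited per owner.
# An IsLocal branch plays LoopA through its own attenuation node and slight volume modulation when true.
# When false or unset, it plays one of two equally weighted Applied loops at boosted volume.
# Envelopers on both paths quickly fade the loop in, then ease its volume down, more steeply on the non-local path.
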SoundCue /VividRazorMedallionsGameplay/Assets/Audio/SC_VividRazor_Unstoppable_Loop1
root=Branch_0
nodes [
 {
  id=Attenuation_2
  children [Modulator_0]
  properties {attenuationSettings=&/Att_VividRazor_Player_Bashable_Loop}
 }
 {id=Branch_0, children [Attenuation_2, Modulator_1, Modulator_1], properties {boolParameterName=IsLocal}}
 {
  id=Enveloper_0
  children [WavePlayer_0]
  properties {pitchCurve {editorCurveData {keys [{time=0}]}}, volumeCurve {editorCurveData {keys [<at 0 0>,<at 0.2 1>,<at 10 0.8>]}}}
 }
 {
  id=Enveloper_1
  children [Random_0]
  properties {pitchCurve {editorCurveData {keys [{time=0}]}}, volumeCurve {editorCurveData {keys [<at 0 0>,<at 0.2 1>,<at 6 0.4>]}}}
 }
 {id=Modulator_0, children [Enveloper_0], properties {volumeMax=1, volumeMin=0.9}}
 {id=Modulator_1, children [Enveloper_1], properties {volumeMax=1.5, volumeMin=1.5}}
 {id=Random_0, children [WavePlayer_7,WavePlayer_6], properties {bShouldExcludeFromBranchCulling=true, weights [1,1]}}
 {id=WavePlayer_0, properties {bLooping=true, soundWaveAssetPtr=&/VividRazor_Unstoppable_LoopA}}
 {
  id=WavePlayer_6
  properties {bLooping=true, soundWaveAssetPtr=&/VividRazor_Unstoppable_Applied_Loop_02}
 }
 {
  id=WavePlayer_7
  properties {bLooping=true, soundWaveAssetPtr=&/VividRazor_Unstoppable_Applied_Loop_03}
 }
]
properties {
 attenuationSettings=&/Att_VividRazor_Unstoppable_Player_Apply
 bOverrideConcurrency=true
 maxDistance=12600
 soundClassObject=&Fort_Gadgets
 attenuationOverrides {
  bAttenuate=false
  bEnableListenerFocus=true
  bEnableReverbSend=false
  bEnableSendToAudioLink=false
  focusAzimuth=15
  focusDistanceScale=0.5
  focusVolumeAttenuation=2
  nonFocusAzimuth=30
  nonFocusDistanceScale=0.5
  nonFocusVolumeAttenuation=2
  stereoSpread=500
 }
 concurrencyOverrides {bLimitToOwner=true, maxCount=4}
}
