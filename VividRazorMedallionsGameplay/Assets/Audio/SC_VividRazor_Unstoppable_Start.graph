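# Sound cue that plays the start sound for the VividRazor medallion's Unstoppable effect, classed as a gadget sound.
# An IsLocal branch sends local playback through a modulator into a random pick of five start waves.
# Non-local and unset cases use a quieter modulated pick of five like-named waves, with attenuation borrowed from the Player Bashable loop.
# The cue has its own Unstoppable apply attenuation and overrides concurrency with a small per-owner voice limit.
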SoundCue /VividRazorMedallionsGameplay/Assets/Audio/SC_VividRazor_Unstoppable_Start
root=Branch_0
nodes [
 {
  id=Attenuation_2
  children [Modulator_1]
  properties {attenuationSettings=&/Att_VividRazor_Player_Bashable_Loop}
 }
 {id=Branch_0, children [Modulator_0, Attenuation_2, Attenuation_2], properties {boolParameterName=IsLocal}}
 {id=Modulator_0, children [Random_0]}
 {id=Modulator_1, children [Random_1], properties {volumeMax=0.6, volumeMin=0.5}}
 {id=Random_0, children [WavePlayer_0,WavePlayer_1,WavePlayer_2,WavePlayer_3,WavePlayer_4], properties {weights [1,1,1,1,1]}}
 {id=Random_1, children [WavePlayer_5,WavePlayer_6,WavePlayer_7,WavePlayer_8,WavePlayer_9], properties {weights [1,1,1,1,1]}}
 {id=WavePlayer_0, properties {soundWaveAssetPtr=&/VividRazor_Unstoppable_Start_01}}
 {id=WavePlayer_1, properties {soundWaveAssetPtr=&/VividRazor_Unstoppable_Start_02}}
 {id=WavePlayer_2, properties {soundWaveAssetPtr=&/VividRazor_Unstoppable_Start_03}}
 {id=WavePlayer_3, properties {soundWaveAssetPtr=&/VividRazor_Unstoppable_Start_04}}
 {id=WavePlayer_4, properties {soundWaveAssetPtr=&/VividRazor_Unstoppable_Start_05}}
 {id=WavePlayer_5, properties {soundWaveAssetPtr=&VividRazor_Unstoppable_Start_01}}
 {id=WavePlayer_6, properties {soundWaveAssetPtr=&VividRazor_Unstoppable_Start_02}}
 {id=WavePlayer_7, properties {soundWaveAssetPtr=&VividRazor_Unstoppable_Start_03}}
 {id=WavePlayer_8, properties {soundWaveAssetPtr=&VividRazor_Unstoppable_Start_04}}
 {id=WavePlayer_9, properties {soundWaveAssetPtr=&VividRazor_Unstoppable_Start_05}}
]
properties {
 attenuationSettings=&/Att_VividRazor_Unstoppable_Player_Apply
 bOverrideConcurrency=true
 maxDistance=12600
 soundClassObject=&Fort_Gadgets
 volumeMultiplier=0.4
 concurrencyOverrides {bLimitToOwner=true, maxCount=4}
}
