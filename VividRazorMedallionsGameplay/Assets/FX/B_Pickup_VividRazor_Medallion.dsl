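; Pickup effect Blueprint for the VividRazor medallion, controlling the visuals of its parent pickup actor.
; On BeginPlay it adds a rotating movement component that ticks during physics and drives the pickup's static mesh, then destroys its inherited or own BG component.
; The idle effect, named after the UnluckyRoll medallion, is hidden when picked up and shown again when tossed.
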(event EventBeginPlay
  (bind _parentpickupactor (Variables|PickupProperties|GetParentPickupActor))
  (|CallParentFunction)
  (bind _returnvalue (AddComponent|NODE_AddRotatingMovementComponent-0 :self _parentpickupactor))
  (Components|Tick|SetTickGroup _returnvalue "TG_DuringPhysics")
  (Components|Movement|SetUpdatedComponent _returnvalue (Class|FortPickup|GetPickupStaticMeshComponent _parentpickupactor))
  (Components|DestroyComponent (Variables|Default|GetBG)))

(event Collision|EventActorBeginOverlap
  (params (OtherActor :type Actor%20Object%20Reference)))

(event EventTick
  (params (DeltaSeconds :type Float%20%28single-precision%29)))

(event EventOnPickedUp
  (params (PickupTarget :type Fort%20Pawn%20Object%20Reference))
  (|CallParentFunction 0)
  (Development|SetHiddeninGame (Variables|Default|GetNS_UnluckyRoll_Medallion_Idle) true true))

(event EventOnTossed
  (|CallParentFunction)
  (Development|SetHiddeninGame (Variables|Default|GetNS_UnluckyRoll_Medallion_Idle) false true))

(fn ConstructionScript
  (|CallParentFunction))
